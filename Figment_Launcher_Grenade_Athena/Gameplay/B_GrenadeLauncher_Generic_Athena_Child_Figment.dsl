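; Figment grenade launcher weapon Blueprint that plays an attached out-of-ammo sound on the weapon mesh when fired empty.
; It works around a mobile bug (FORT-1022261) where holding fire retriggers that sound repeatedly.
; On mobile the sound goes through a DoOnce gate, which is also wired to a local ammo-changed event bound at BeginPlay that checks for a loaded magazine.
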(event EventTick
  (params (DeltaSeconds :type Float%20%28single-precision%29)))

(event EventOnOutOfAmmoFire
  ; FORT-1022261 - Workaround for Out of ammo sound repeatedly triggering when holding the fire button on Mobile
  (if (TouchPlatform|IsMobileGame)
    (Utilities|FlowControl|DoOnce false
      (:Completed
        _)))
  (Audio|SpawnSoundAttached "/Game/Sounds/Fort_Weapons/OutofAmmo/Launcher_OutofAmmo_Cue.Launcher_OutofAmmo_Cue" (Variables|Visual|GetWeaponMesh)))

(event EventBeginPlay
  ; FORT-1022261 - Workaround for Out of ammo sound repeatedly triggering when holding the fire button on Mobile
  (EventDispatchers|AssignOnLocalAmmoChanged (delegate Custom|OnLocalAmmoChanged_Event)))

(event Custom|OnLocalAmmoChanged_Event
  (params (LocalCount :type Integer) (LocalRemaining :type Integer))
  ; FORT-1022261 - Workaround for Out of ammo sound repeatedly triggering when holding the fire button on Mobile
  (if (>= (FortWeapon|GetLocalMagazineAmmoCount) 1)
    (Utilities|FlowControl|DoOnce false
      (:Completed
        (Audio|SpawnSoundAttached "/Game/Sounds/Fort_Weapons/OutofAmmo/Launcher_OutofAmmo_Cue.Launcher_OutofAmmo_Cue" (Variables|Visual|GetWeaponMesh))))))

(fn ConstructionScript
  (|CallParentFunction))
